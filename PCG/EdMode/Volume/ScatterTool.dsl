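; PCG scatter tool graph for the editor mode's paint tool, spawning weighted static meshes or actors on landscape or ray-hit surfaces.
; Surface samples are gathered with paint-layer points, then thinned by probability or fixed count.
; Points get weighted assets, global and random transforms, optional self-pruning and surface projection, then subtraction of the exclusion layer.
; With no assets, tagged points render as debug previews; optional debug branches show brush stamps and painted bounds.
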(graph ScatterTool
  (params (PointsPerSquaredMeter :type Double :default 0.1) (PointSize :type Vector2 :default (1e+02 1e+02)) (Uniformity :type Double) (SamplingSeed :type Int :default 492183294) (SampleComplexCollision :type Bool) (SampleLandscapeOnly :type Bool) (ProbabilityFilter :type Double :default 1e+02) (UseFixedCountFilter :type Bool) (FixedCountFilter :type Int :default 10) (FilterSeed :type Int :default -193199241) (RemoveOverlapingPoints :type Bool) (UseMeshBounds :type Bool) (StaticMeshes :type (Array (Struct /PCG/EdMode/Resources/PCGEdMode_WeightedStaticMeshReference.PCGEdMode_WeightedStaticMeshReference))) (SpawnActors :type Bool) (ActorClasses :type (Array (Struct /PCG/EdMode/Resources/PCGEdMode_WeightedClassReference.PCGEdMode_WeightedClassReference))) (GlobalPositionOffset :type Vector) (GlobalRotation :type Rotator) (AbsoluteGlobalRotation :type Bool) (GlobalScaleFactor :type Vector :default (1.0 1.0 1.0)) (RandomPositionOffsetMin :type Vector) (RandomPositionOffsetMax :type Vector) (RandomRotationMin :type Rotator) (RandomRotationMax :type Rotator) (RandomScaleFactorMin :type Vector :default (1.0 1.0 1.0)) (RandomScaleFactorMax :type Vector :default (1.0 1.0 1.0)) (UniformRandomScaleFactor :type Bool :default true) (RandomSeed :type Int :default 2118139241) (SnapToSurface :type Bool) (AlignToSurface :type Bool) (ShowBrushStamps :type Bool) (ShowPaintedDataBounds :type Bool))
  (bind Select_1 (Select :bUseInputB-param SpawnActors :Input%20A StaticMeshes :Input%20B ActorClasses))
  (bind RemoveEmptyData_1 (Remove%20Empty%20Data))
  (bind DataNum_1 (Data%20Count))
  (bind AttributeMathsOp_6 (Attribute%20Maths%20Op :Operation Divide :InA ProbabilityFilter))
  (bind AttributeMathsOp_0 (Attribute%20Maths%20Op :Operation OneMinus :In Uniformity))
  (bind AttributeMathsOp_4 (Attribute%20Maths%20Op :Operation Divide :InA PointSize))
  (bind AttributeMathsOp_3 (Attribute%20Maths%20Op :Operation Max))
  (bind SurfaceSampler_0 (Surface%20Sampler :bApplyDensityToPoints false :Looseness 0.0 :PointsPerSquaredMeter-param PointsPerSquaredMeter :Seed-param SamplingSeed))
  (bind RemoveEmptyData_0 (Remove%20Empty%20Data))
  (bind Gather_1 (Gather))
  (bind Branch_4 (Branch :bOutputToB-param UseFixedCountFilter))
  (bind AttributeNoise_0 (Attribute%20Noise :Seed-param FilterSeed))
  (bind AttributeFilter_0 (Filter%20Attribute%20Elements :bGenerateOutputDataEvenIfEmpty false :Operator Lesser :ThresholdAttribute @Last))
  (bind RandomChoice_0 (Random%20Choice :FixedNumber 10 :FixedNumber-param FixedCountFilter :Seed-param FilterSeed))
  (bind Branch_0 (Branch))
  (bind MatchAndSetAttributes_1 (Match%20And%20Set%20Attributes :bUseWeightAttribute true :WeightAttribute Weight))
  (bind AddTags_0 (Add%20Tags :TagsToAdd "ShowPreviewMesh"))
  (bind Branch_2 (Branch :bOutputToB-param SpawnActors))
  (bind Branch_5 (Branch :bOutputToB-param UseMeshBounds))
  (bind BoundsFromMesh_1 (Bounds%20From%20Mesh :MeshAttribute StaticMesh))
  (bind TransformPoints_1 (Transform%20Points :bUniformScale false :bAbsoluteRotation-param AbsoluteGlobalRotation :OffsetMax-param GlobalPositionOffset :OffsetMin-param GlobalPositionOffset :RotationMax-param GlobalRotation :RotationMin-param GlobalRotation :ScaleMax-param GlobalScaleFactor :ScaleMin-param GlobalScaleFactor))
  (bind TransformPoints_2 (Transform%20Points :bUniformScale-param UniformRandomScaleFactor :OffsetMax-param RandomPositionOffsetMax :OffsetMin-param RandomPositionOffsetMin :RotationMax-param RandomRotationMax :RotationMin-param RandomRotationMin :ScaleMax-param RandomScaleFactorMax :ScaleMin-param RandomScaleFactorMin :Seed-param RandomSeed))
  (bind Branch_6 (Branch :bOutputToB-param RemoveOverlapingPoints))
  (bind SelfPruning_0 (Self%20Pruning :Parameters (struct :ComparisonSource $Extents :bRandomizedPruning false)))
  (bind Projection_1 (Projection :bKeepZeroDensityPoints true :ProjectionParams (struct :AttributeMode IncludeAttributes :AttributeMergeOperation SourceValue) :bProjectPositions SnapToSurface :bProjectRotations AlignToSurface))
  (bind Difference_0 (Difference :bDiffMetadata false :DensityFunction Binary :Mode Discrete))
  (bind FilterDataByTag_0 (Filter%20Data%20By%20Tag :SelectedTags "ShowPreviewMesh"))
  (bind Branch_1 (Branch :bOutputToB-param SpawnActors))
  (bind StaticMeshSpawner_1 (Static%20Mesh%20Spawner :MeshSelectorParameters (class:/Script/PCG.PCGMeshSelectorByAttribute :AttributeName StaticMesh :TemplateDescriptor (struct :bUseDefaultCollision true)) :MeshSelectorType /Script/PCG.PCGMeshSelectorByAttribute))
  (bind SpawnActor_1 (Spawn%20Actor :bSpawnByAttribute true :PostSpawnFunctionNames [None] :SpawnAttribute Class))
  (bind Debug_0 (Debug))
  (bind Branch_7 (Branch :bOutputToB-param SampleLandscapeOnly :In SampleLandscapeOnly))
  (bind WorldRayHitQuery_0 (World%20Ray%20Hit%20Query :bTraceComplex SampleComplexCollision))
  (bind GetLandscapeData_0 (Get%20Landscape%20Data :bUnbounded false :SamplingProperties (struct :bGetLayerWeights false :bSampleVirtualTextures false)))
  (bind Gather_0 (Gather))
  (bind GetToolData_0 (Get%20Tool%20Data :DataInstance PaintLayer :ToolTag PaintTool))
  (bind GetToolData_1 (Get%20Tool%20Data :DataInstance ExclusionLayer :ToolTag PaintTool))
  (bind FilterDataByType_2 (Filter%20Data%20By%20Type :TargetType (struct :Ids [(struct :Struct /Script/PCG.PCGDataTypeInfoPoint)])))
  (bind FilterDataByType_0 (Filter%20Data%20By%20Type :TargetType (struct :Ids [(struct :Struct /Script/PCG.PCGDataTypeInfoPoint)])))
  ;[1 0.1333 0] Debug Paint Data
  (bind Branch_8 (Branch :bOutputToB-param ShowBrushStamps))
  (bind AttributeMathsOp_1 (Attribute%20Maths%20Op :InputSource1 $Extents :Operation Divide :OutputTarget $Scale))
  (bind Debug_1 (Debug))
  (bind Branch_9 (Branch :bOutputToB-param ShowPaintedDataBounds))
  (bind MergePoints_1 (Merge%20Points :bMergeMetadata false))
  (bind CombinePoints_0 (Combine%20Points))
  (bind AttributeMathsOp_2 (Attribute%20Maths%20Op :InputSource1 $Extents :Operation Divide :OutputTarget $Scale))
  (bind Debug_2 (Debug))
  (bind Branch_10 (Branch :bOutputToB-param ShowBrushStamps))
  (bind AttributeMathsOp_7 (Attribute%20Maths%20Op :InputSource1 $Extents :Operation Divide :OutputTarget $Scale))
  (bind Debug_3 (Debug))
  (reroute Exclusion%20Layer%20Paint%20Data FilterDataByType_2.InsideFilter)
  (reroute PaintLayerData FilterDataByType_0.InsideFilter)
  (reroute Surface Gather_0)
  (connect Branch_0 Output%20A AddTags_0 In)
  (connect AttributeMathsOp_6 Out AttributeFilter_0 Filter)
  (connect AttributeNoise_0 Out AttributeFilter_0 In)
  (connect Branch_8 Output%20B AttributeMathsOp_1 InA)
  (connect CombinePoints_0 Out AttributeMathsOp_2 InA)
  (connect AttributeMathsOp_4 Out AttributeMathsOp_3 InA)
  (connect Branch_10 Output%20B AttributeMathsOp_7 InA)
  (connect Branch_4 Output%20A AttributeNoise_0 In)
  (connect Branch_5 Output%20B BoundsFromMesh_1 In)
  (connect DataNum_1 Out Branch_0 bOutputToB)
  (connect AttributeFilter_0 InsideFilter Branch_0 In)
  (connect RandomChoice_0 Chosen Branch_0 In)
  (connect FilterDataByTag_0 OutsideFilter Branch_1 In)
  (connect MatchAndSetAttributes_1 Out Branch_2 In)
  (connect Gather_1 Out Branch_4 In)
  (connect Branch_2 Output%20A Branch_5 In)
  (connect TransformPoints_2 Out Branch_6 In)
  (connect Exclusion%20Layer%20Paint%20Data Out Branch_8 In)
  (connect Exclusion%20Layer%20Paint%20Data Out Branch_9 In)
  (connect PaintLayerData Out Branch_9 In)
  (connect PaintLayerData Out Branch_10 In)
  (connect MergePoints_1 Out CombinePoints_0 In)
  (connect RemoveEmptyData_1 Out DataNum_1 In)
  (connect FilterDataByTag_0 InsideFilter Debug_0 In)
  (connect AttributeMathsOp_1 Out Debug_1 In)
  (connect AttributeMathsOp_2 Out Debug_2 In)
  (connect AttributeMathsOp_7 Out Debug_3 In)
  (connect Exclusion%20Layer%20Paint%20Data Out Difference_0 Differences)
  (connect Projection_1 Out Difference_0 Source)
  (connect Difference_0 Out FilterDataByTag_0 In)
  (connect GetToolData_0 Out FilterDataByType_0 In)
  (connect GetToolData_1 Out FilterDataByType_2 In)
  (connect WorldRayHitQuery_0 Out Gather_0 In)
  (connect GetLandscapeData_0 Out Gather_0 In2)
  (connect RemoveEmptyData_0 Out Gather_1 In)
  (connect PaintLayerData Out Gather_1 In2)
  (connect Branch_7 Output%20B GetLandscapeData_0 Execution%20Dependency)
  (connect Branch_0 Output%20B MatchAndSetAttributes_1 In)
  (connect Select_1 Out MatchAndSetAttributes_1 Match%20Data)
  (connect Branch_9 Output%20B MergePoints_1 In)
  (connect Branch_6 Output%20A Projection_1 In)
  (connect SelfPruning_0 Out Projection_1 In)
  (connect Surface Out Projection_1 Projection%20Target)
  (connect Branch_4 Output%20B RandomChoice_0 In)
  (connect SurfaceSampler_0 Out RemoveEmptyData_0 In)
  (connect Select_1 Out RemoveEmptyData_1 In)
  (connect Branch_6 Output%20B SelfPruning_0 In)
  (connect Branch_1 Output%20B SpawnActor_1 In)
  (connect Branch_1 Output%20A StaticMeshSpawner_1 In)
  (connect AttributeMathsOp_0 Out SurfaceSampler_0 Looseness)
  (connect AttributeMathsOp_3 Out SurfaceSampler_0 PointExtents)
  (connect Surface Out SurfaceSampler_0 Surface)
  (connect AddTags_0 Out TransformPoints_1 In)
  (connect BoundsFromMesh_1 Out TransformPoints_1 In)
  (connect Branch_2 Output%20B TransformPoints_1 In)
  (connect Branch_5 Output%20A TransformPoints_1 In)
  (connect TransformPoints_1 Out TransformPoints_2 In)
  (connect Branch_7 Output%20A WorldRayHitQuery_0 Execution%20Dependency))
